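# Sound cue for the Puppeteer boss's melee swipe attack sound effect.
# A Random node picks between two equally weighted melee attack waves, then a Modulator varies the playback.
# It uses close-range stereo attenuation, a melee weapon sound class, boosted volume and a concurrency override capping simultaneous instances.
# A distance-scaled send feeds a music-volume sidechain audio bus, apparently to duck music.
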SoundCue /PuppeteerBoss/Audio/SFX/Boss/SC_Puppet_Boss_Swipe_Attack
root=Modulator_0
nodes [
 {id=Modulator_0, children [Random_0]}
 {id=Random_0, children [WavePlayer_0,WavePlayer_1], properties {weights [1,1]}}
 {id=WavePlayer_0, properties {soundWaveAssetPtr=/PuppeteerBoss/Audio/SFX/Boss/wav/PuppeteerBoss_Melee_Atk_01}}
 {id=WavePlayer_1, properties {soundWaveAssetPtr=/PuppeteerBoss/Audio/SFX/Boss/wav/PuppeteerBoss_Melee_Atk_02}}
]
properties {
 attenuationSettings=&/CS_Puppet_Boss_Melee_Attack_Close_Range_Stereo_Att
 bOverrideConcurrency=true
 maxDistance=5600
 soundClassObject=&Fort_WeaponMelee_PSM
 volumeMultiplier=2
 busSends [{
   audioBus=&AB_Sidechain_FortMusic_Vol
   maxSendLevel=0
   minSendDistance=300
   minSendLevel=0.3
   sendLevel=0.3
   soundSourceBus=null
   sourceBusSendLevelControlMethod=Linear
  }]
 concurrencyOverrides {maxCount=3}
}
