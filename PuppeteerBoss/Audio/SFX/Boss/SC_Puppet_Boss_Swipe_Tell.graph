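# Sound cue for the Puppeteer boss's melee swipe tell, the audio warning that precedes the attack.
# A Random node picks between two equally weighted melee attack tell waves and feeds a Modulator that adds slight pitch and volume variation.
# It uses close-range stereo melee attenuation, a weapon melee sound class, a boosted volume multiplier, and a concurrency override that caps simultaneous instances.
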SoundCue /PuppeteerBoss/Audio/SFX/Boss/SC_Puppet_Boss_Swipe_Tell
root=Modulator_0
nodes [
 {id=Modulator_0, children [Random_1], properties {pitchMax=1, pitchMin=0.98, volumeMax=1.1, volumeMin=1}}
 {id=Random_1, children [WavePlayer_3,WavePlayer_4], properties {weights [1,1]}}
 {id=WavePlayer_3, properties {soundWaveAssetPtr=/PuppeteerBoss/Audio/SFX/Boss/wav/PuppeteerBoss_Melee_Atk_Tell_01}}
 {id=WavePlayer_4, properties {soundWaveAssetPtr=/PuppeteerBoss/Audio/SFX/Boss/wav/PuppeteerBoss_Melee_Atk_Tell_02}}
]
properties {
 attenuationSettings=&/CS_Puppet_Boss_Melee_Attack_Close_Range_Stereo_Att
 bOverrideConcurrency=true
 maxDistance=5600
 soundClassObject=&Fort_WeaponMelee_PSM
 volumeMultiplier=1.5
 concurrencyOverrides {maxCount=3}
}
